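; Mobile vendor commerce prop Blueprint whose construction script tints its secondary meshes from building texture data.
; After calling the parent script, it gathers all static mesh components and removes the main StaticMeshComponent from the list.
; It converts the tint color of the last building texture data entry to a linear color and applies it to each remaining mesh through SetColoronSecondaryMesh.
; BeginPlay, Tick and overlap events are empty.
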(event EventBeginPlay)

(event Collision|EventActorBeginOverlap
  (params (OtherActor :type Actor%20Object%20Reference)))

(event EventTick
  (params (DeltaSeconds :type Float%20%28single-precision%29)))

(fn ConstructionScript
  (|CallParentFunction)
  (bind _outtexturedataarray (Building|GetBuildingTextureData))
  (bind _outtexturedataarray_2 (Building|GetBuildingTextureData))
  (bind _output_get (LocalVariables|SetL_AR_MeshComponents (Actor|GetComponentsByClass "/Script/Engine.StaticMeshComponent")))
  (Utilities|Array|RemoveItem _output_get (Variables|Debug|GetStaticMeshComponent))
  (for _array_element (LocalVariables|GetL_AR_MeshComponents)
    (Default|SetColoronSecondaryMesh (Math|Conversions|ToLinearColor(Color) (Variables|/Script/FortniteGame.BuildingTextureData|GetTintColor (Utilities|Array|Get(acopy) _outtexturedataarray_2 (- (Utilities|Array|Length _outtexturedataarray_2) 1)))) _array_element)))
